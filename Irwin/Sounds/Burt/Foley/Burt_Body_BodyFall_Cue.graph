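# Sound cue for Burt's body-fall foley, a creature sound that plays one of several short body-fall waves at random.
# A quality-level switch sends the first two levels to a four-wave random pool and the last two to a two-wave subset, the latter excluded from branch culling.
# It uses shared prey gallop attenuation, the creature foley sound class, and owner-limited and prey-group concurrency sets.
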SoundCue /Irwin/Sounds/Burt/Foley/Burt_Body_BodyFall_Cue
root=QualityLevel_0
nodes [
 {id=QualityLevel_0, children [Random_2,Random_2,Random_0,Random_0]}
 {id=Random_0, children [WavePlayer_0,WavePlayer_3], properties {bShouldExcludeFromBranchCulling=true, weights [1,1]}}
 {id=Random_2, children [WavePlayer_11, WavePlayer_12, WavePlayer_13, WavePlayer_14], properties {weights [1,1,1,1]}}
 {id=WavePlayer_0, properties {soundWaveAssetPtr=&/Burt_Body_BodyFall_Short_01}}
 {id=WavePlayer_11, properties {soundWaveAssetPtr=&Burt_Body_BodyFall_Short_01}}
 {id=WavePlayer_12, properties {soundWaveAssetPtr=&/Burt_Body_BodyFall_Short_02}}
 {id=WavePlayer_13, properties {soundWaveAssetPtr=&/Burt_Body_BodyFall_Short_03}}
 {id=WavePlayer_14, properties {soundWaveAssetPtr=&/Burt_Body_BodyFall_Short_04}}
 {id=WavePlayer_3, properties {soundWaveAssetPtr=&Burt_Body_BodyFall_Short_04}}
]
properties {
 attenuationSettings=/Irwin/Sounds/Settings/Prey_Gallop_Att
 maxDistance=3200
 soundClassObject=/Game/Sounds/FortSoundClassesAndModes/Fort_Creatures_Foley
 volumeMultiplier=0.65
 concurrencySet [/Irwin/Sounds/Settings/Conc/SetOneLimitToOwner_Foley_Conc, /Irwin/Sounds/Settings/Conc/Prey_SetGroup_Foley_Conc]
}
