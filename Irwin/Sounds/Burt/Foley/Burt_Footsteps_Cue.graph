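# Sound cue for Burt's footsteps, routed by whether the listener is local and whether Burt is ridden by an enemy.
# IsLocal and IsRiddenByEnemy branches pick sound class and attenuation, with enemy-ridden steps louder than non-enemy ones and using enemy foley class and riding attenuation.
# All routes share one pitch and volume randomizer feeding a quality switch between eleven-wave and four-wave random pools.
# A Prey foley concurrency set limits playback.
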SoundCue /Irwin/Sounds/Burt/Foley/Burt_Footsteps_Cue
root=Branch_2
nodes [
 {id=Attenuation_0, children [Modulator_1], properties {attenuationSettings=/Irwin/Sounds/Settings/Prey_Step_Att}}
 {id=Attenuation_2, children [Modulator_1], properties {attenuationSettings=/Game/Sounds/Attenuation/Fort_Riding_Enemy_Att}}
 {id=Attenuation_3, children [Modulator_1], properties {attenuationSettings=&Prey_Step_Att}}
 {id=Attenuation_4, children [Modulator_1], properties {attenuationSettings=&Prey_Step_Att}}
 {id=Branch_1, children [Modulator_3,Modulator_2,SoundClass_0], properties {boolParameterName=IsRiddenByEnemy}}
 {id=Branch_2, children [SoundClass_4,Branch_1,Branch_1], properties {boolParameterName=IsLocal}}
 {id=Modulator_1, children [QualityLevel_0], properties {pitchMax=1.1, pitchMin=0.9, volumeMax=0.9, volumeMin=0.5}}
 {id=Modulator_2, children [SoundClass_3], properties {pitchMax=1, pitchMin=1, volumeMax=0.3, volumeMin=0.3}}
 {id=Modulator_3, children [SoundClass_2], properties {pitchMax=1, pitchMin=1, volumeMax=0.7, volumeMin=0.7}}
 {id=QualityLevel_0, children [Random_1,Random_1,Random_2,Random_2]}
 {
  id=Random_1
  children [
   WavePlayer_0
   WavePlayer_9
   WavePlayer_10
   WavePlayer_11
   WavePlayer_12
   WavePlayer_13
   WavePlayer_14
   WavePlayer_15
   WavePlayer_16
   WavePlayer_17
   WavePlayer_18
  ]
  properties {weights [1,1,1,1,1,1,1,1,1,1,1]}
 }
 {
  id=Random_2
  children [WavePlayer_1,WavePlayer_2,WavePlayer_3,WavePlayer_4]
  properties {bShouldExcludeFromBranchCulling=true, weights [1,1,1,1]}
 }
 {
  id=SoundClass_0
  children [Attenuation_0]
  properties {soundClassOverride=/Game/Sounds/FortSoundClassesAndModes/Fort_Creatures_Steps_Tactical}
 }
 {
  id=SoundClass_2
  children [Attenuation_2]
  properties {soundClassOverride=&Fort_Player_Foley_3P_Enemy_Athena}
 }
 {
  id=SoundClass_3
  children [Attenuation_3]
  properties {soundClassOverride=&Fort_Creatures_Steps_Tactical}
 }
 {
  id=SoundClass_4
  children [Attenuation_4]
  properties {soundClassOverride=&Fort_Creatures_Steps_Tactical}
 }
 {id=WavePlayer_0, properties {soundWaveAssetPtr=&/Burt_Footsteps_10}}
 {id=WavePlayer_1, properties {soundWaveAssetPtr=&Burt_Footsteps_10}}
 {id=WavePlayer_10, properties {soundWaveAssetPtr=&/Burt_Footsteps_01}}
 {id=WavePlayer_11, properties {soundWaveAssetPtr=&/Burt_Footsteps_02}}
 {id=WavePlayer_12, properties {soundWaveAssetPtr=&/Burt_Footsteps_03}}
 {id=WavePlayer_13, properties {soundWaveAssetPtr=&/Burt_Footsteps_04}}
 {id=WavePlayer_14, properties {soundWaveAssetPtr=&/Burt_Footsteps_05}}
 {id=WavePlayer_15, properties {soundWaveAssetPtr=&/Burt_Footsteps_06}}
 {id=WavePlayer_16, properties {soundWaveAssetPtr=&/Burt_Footsteps_07}}
 {id=WavePlayer_17, properties {soundWaveAssetPtr=&/Burt_Footsteps_08}}
 {id=WavePlayer_18, properties {soundWaveAssetPtr=&/Burt_Footsteps_09}}
 {id=WavePlayer_2, properties {soundWaveAssetPtr=&/Burt_Footsteps_11}}
 {id=WavePlayer_3, properties {soundWaveAssetPtr=&Burt_Footsteps_01}}
 {id=WavePlayer_4, properties {soundWaveAssetPtr=&Burt_Footsteps_02}}
 {id=WavePlayer_9, properties {soundWaveAssetPtr=&Burt_Footsteps_11}}
]
properties {maxDistance=13000, volumeMultiplier=0.8, concurrencySet [/Irwin/Sounds/Settings/Conc/Prey_SetGroup_Foley_Conc]}
