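# Foley SoundCue for the creature Burt's jump, mixing a randomized land plus heavy-land layer with a slide layer.
# Slide variant pools switch by quality level, and both layers get random pitch and volume.
# Each layer branches on IsLocal, then IsEnemy, to set volume, sound class and attenuation, with riding-enemy attenuation and enemy foley class for enemies.
# Concurrency is overridden to cap simultaneous instances.
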SoundCue /Irwin/Sounds/Burt/Foley/Burt_Jump_Cue
root=Mixer_1
nodes [
 {id=Attenuation_0, children [Modulator_0], properties {attenuationSettings=/Game/Sounds/Attenuation/Fort_Riding_Enemy_Att}}
 {id=Attenuation_1, children [Modulator_0], properties {attenuationSettings=/Irwin/Sounds/Settings/Prey_Gallop_Att}}
 {id=Attenuation_2, children [Modulator_0], properties {attenuationSettings=&Prey_Gallop_Att}}
 {id=Attenuation_3, children [Modulator_0], properties {attenuationSettings=&Prey_Gallop_Att}}
 {id=Attenuation_4, children [Modulator_4], properties {attenuationSettings=&Fort_Riding_Enemy_Att}}
 {id=Attenuation_5, children [Modulator_4], properties {attenuationSettings=/Irwin/Sounds/Settings/Prey_Step_Att}}
 {id=Attenuation_6, children [Modulator_4], properties {attenuationSettings=&Prey_Step_Att}}
 {id=Attenuation_7, children [Modulator_4], properties {attenuationSettings=&Prey_Step_Att}}
 {id=Branch_0, children [Modulator_2,Modulator_1,SoundClass_3], properties {boolParameterName=IsEnemy}}
 {id=Branch_1, children [Modulator_3,Branch_0,Branch_0], properties {boolParameterName=IsLocal}}
 {id=Branch_2, children [Modulator_6,Modulator_5,SoundClass_7], properties {boolParameterName=IsEnemy}}
 {id=Branch_3, children [Modulator_7,Branch_2,Branch_2], properties {boolParameterName=IsLocal}}
 {id=Mixer_0, children [Random_0,Random_2], properties {inputVolume [1,1]}}
 {id=Mixer_1, children [Branch_1,Branch_3], properties {inputVolume [0.7,1.25]}}
 {id=Modulator_0, children [Mixer_0], properties {pitchMax=1.2, pitchMin=0.9, volumeMax=1.15, volumeMin=0.85}}
 {id=Modulator_1, children [SoundClass_1], properties {pitchMax=1, pitchMin=1, volumeMax=0.1, volumeMin=0.1}}
 {id=Modulator_2, children [SoundClass_0], properties {pitchMax=1, pitchMin=1, volumeMax=0.4, volumeMin=0.4}}
 {id=Modulator_3, children [SoundClass_2], properties {pitchMax=1, pitchMin=1, volumeMax=1.5, volumeMin=1.5}}
 {id=Modulator_4, children [QualityLevel_0], properties {pitchMax=1.1, pitchMin=0.9, volumeMax=0.9, volumeMin=0.75}}
 {id=Modulator_5, children [SoundClass_5], properties {pitchMax=1, pitchMin=1, volumeMax=0.1, volumeMin=0.1}}
 {id=Modulator_6, children [SoundClass_4], properties {pitchMax=1, pitchMin=1, volumeMax=0.4, volumeMin=0.4}}
 {id=Modulator_7, children [SoundClass_6], properties {pitchMax=1, pitchMin=1, volumeMax=1.8, volumeMin=1.5}}
 {id=QualityLevel_0, children [Random_1,Random_1,Random_3,Random_3]}
 {
  id=Random_0
  children [WavePlayer_1,WavePlayer_2,WavePlayer_3,WavePlayer_4,WavePlayer_5]
  properties {bShouldExcludeFromBranchCulling=true, weights [1,1,1,1,1]}
 }
 {id=Random_1, children [WavePlayer_0, WavePlayer_6, WavePlayer_12, WavePlayer_13, WavePlayer_14], properties {weights [1,1,1,1,1]}}
 {
  id=Random_2
  children [WavePlayer_7, WavePlayer_8, WavePlayer_9, WavePlayer_10, WavePlayer_11]
  properties {bShouldExcludeFromBranchCulling=true, weights [1,1,1,1,1]}
 }
 {id=Random_3, children [WavePlayer_15, WavePlayer_16, WavePlayer_17], properties {bShouldExcludeFromBranchCulling=true, weights [1,1,1]}}
 {
  id=SoundClass_0
  children [Attenuation_0]
  properties {soundClassOverride=&Fort_Player_Foley_3P_Enemy_Athena}
 }
 {
  id=SoundClass_1
  children [Attenuation_1]
  properties {soundClassOverride=/Game/Sounds/FortSoundClassesAndModes/Fort_Creatures_Steps_Tactical}
 }
 {
  id=SoundClass_2
  children [Attenuation_2]
  properties {soundClassOverride=&Fort_Creatures_Steps_Tactical}
 }
 {
  id=SoundClass_3
  children [Attenuation_3]
  properties {soundClassOverride=&Fort_Creatures_Steps_Tactical}
 }
 {
  id=SoundClass_4
  children [Attenuation_4]
  properties {soundClassOverride=&Fort_Player_Foley_3P_Enemy_Athena}
 }
 {
  id=SoundClass_5
  children [Attenuation_5]
  properties {soundClassOverride=&Fort_Creatures_Steps_Tactical}
 }
 {
  id=SoundClass_6
  children [Attenuation_6]
  properties {soundClassOverride=&Fort_Creatures_Steps_Tactical}
 }
 {
  id=SoundClass_7
  children [Attenuation_7]
  properties {soundClassOverride=&Fort_Creatures_Steps_Tactical}
 }
 {id=WavePlayer_0, properties {soundWaveAssetPtr=&/Burt_Slide_01}}
 {id=WavePlayer_1, properties {soundWaveAssetPtr=&/Burt_Land_01}}
 {id=WavePlayer_10, properties {soundWaveAssetPtr=&/Burt_Land_Heavy_04}}
 {id=WavePlayer_11, properties {soundWaveAssetPtr=&/Burt_Land_Heavy_05}}
 {id=WavePlayer_12, properties {soundWaveAssetPtr=&/Burt_Slide_03}}
 {id=WavePlayer_13, properties {soundWaveAssetPtr=&/Burt_Slide_04}}
 {id=WavePlayer_14, properties {soundWaveAssetPtr=&/Burt_Slide_05}}
 {id=WavePlayer_15, properties {soundWaveAssetPtr=&Burt_Slide_01}}
 {id=WavePlayer_16, properties {soundWaveAssetPtr=&/Burt_Slide_02}}
 {id=WavePlayer_17, properties {soundWaveAssetPtr=&Burt_Slide_05}}
 {id=WavePlayer_2, properties {soundWaveAssetPtr=&/Burt_Land_02}}
 {id=WavePlayer_3, properties {soundWaveAssetPtr=&/Burt_Land_03}}
 {id=WavePlayer_4, properties {soundWaveAssetPtr=&/Burt_Land_04}}
 {id=WavePlayer_5, properties {soundWaveAssetPtr=&/Burt_Land_05}}
 {id=WavePlayer_6, properties {soundWaveAssetPtr=&Burt_Slide_02}}
 {id=WavePlayer_7, properties {soundWaveAssetPtr=&/Burt_Land_Heavy_01}}
 {id=WavePlayer_8, properties {soundWaveAssetPtr=&/Burt_Land_Heavy_02}}
 {id=WavePlayer_9, properties {soundWaveAssetPtr=&/Burt_Land_Heavy_03}}
]
properties {
 bOverrideConcurrency=true
 maxDistance=13000
 volumeMultiplier=1
 concurrencyOverrides {maxCount=3}
 concurrencySet [/Irwin/Sounds/Settings/Conc/SetOneLimitToOwner_Foley_Conc, /Irwin/Sounds/Settings/Conc/Prey_SetGroup_Foley_Conc]
}
