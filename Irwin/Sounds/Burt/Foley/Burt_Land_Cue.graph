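# Sound cue for Burt's landing foley, layering a random land sound with a random heavy land sound under slight pitch and volume variation.
# IsLocal boosts the local case; otherwise IsRiddenByEnemy selects an enemy-rider mix, a much quieter non-enemy mix, or an unscaled fallback when unset.
# Each path sets its own sound class and attenuation around the shared layered core.
# Concurrency sets for per-owner and prey-group foley limit overlapping plays.
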SoundCue /Irwin/Sounds/Burt/Foley/Burt_Land_Cue
root=Branch_1
nodes [
 {id=Attenuation_0, children [Modulator_0], properties {attenuationSettings=/Game/Sounds/Attenuation/Fort_Riding_Enemy_Att}}
 {id=Attenuation_1, children [Modulator_0], properties {attenuationSettings=/Irwin/Sounds/Settings/Prey_Gallop_Att}}
 {id=Attenuation_2, children [Modulator_0], properties {attenuationSettings=&Prey_Gallop_Att}}
 {id=Attenuation_3, children [Modulator_0], properties {attenuationSettings=&Prey_Gallop_Att}}
 {id=Branch_0, children [Modulator_2,Modulator_1,SoundClass_3], properties {boolParameterName=IsRiddenByEnemy}}
 {id=Branch_1, children [Modulator_3,Branch_0,Branch_0], properties {boolParameterName=IsLocal}}
 {id=Mixer_0, children [Random_0,Random_2], properties {inputVolume [1.2,1]}}
 {id=Modulator_0, children [Mixer_0], properties {pitchMax=1.2, pitchMin=0.9, volumeMax=1.15, volumeMin=0.85}}
 {id=Modulator_1, children [SoundClass_1], properties {pitchMax=1, pitchMin=1, volumeMax=0.1, volumeMin=0.1}}
 {id=Modulator_2, children [SoundClass_0], properties {pitchMax=1, pitchMin=1, volumeMax=0.55, volumeMin=0.55}}
 {id=Modulator_3, children [SoundClass_2], properties {pitchMax=1, pitchMin=1, volumeMax=1.5, volumeMin=1.5}}
 {
  id=Random_0
  children [WavePlayer_1,WavePlayer_2,WavePlayer_3,WavePlayer_4,WavePlayer_5]
  properties {bShouldExcludeFromBranchCulling=true, weights [1,1,1,1,1]}
 }
 {
  id=Random_2
  children [WavePlayer_7, WavePlayer_8, WavePlayer_9, WavePlayer_10, WavePlayer_11]
  properties {bShouldExcludeFromBranchCulling=true, weights [1,1,1,1,1]}
 }
 {
  id=SoundClass_0
  children [Attenuation_0]
  properties {soundClassOverride=&Fort_Player_Foley_3P_Enemy_Athena}
 }
 {
  id=SoundClass_1
  children [Attenuation_1]
  properties {soundClassOverride=/Game/Sounds/FortSoundClassesAndModes/Fort_Creatures_Steps_Tactical}
 }
 {
  id=SoundClass_2
  children [Attenuation_2]
  properties {soundClassOverride=&Fort_Creatures_Steps_Tactical}
 }
 {
  id=SoundClass_3
  children [Attenuation_3]
  properties {soundClassOverride=&Fort_Creatures_Steps_Tactical}
 }
 {id=WavePlayer_1, properties {soundWaveAssetPtr=&/Burt_Land_01}}
 {id=WavePlayer_10, properties {soundWaveAssetPtr=&/Burt_Land_Heavy_04}}
 {id=WavePlayer_11, properties {soundWaveAssetPtr=&/Burt_Land_Heavy_05}}
 {id=WavePlayer_2, properties {soundWaveAssetPtr=&/Burt_Land_02}}
 {id=WavePlayer_3, properties {soundWaveAssetPtr=&/Burt_Land_03}}
 {id=WavePlayer_4, properties {soundWaveAssetPtr=&/Burt_Land_04}}
 {id=WavePlayer_5, properties {soundWaveAssetPtr=&/Burt_Land_05}}
 {id=WavePlayer_7, properties {soundWaveAssetPtr=&/Burt_Land_Heavy_01}}
 {id=WavePlayer_8, properties {soundWaveAssetPtr=&/Burt_Land_Heavy_02}}
 {id=WavePlayer_9, properties {soundWaveAssetPtr=&/Burt_Land_Heavy_03}}
]
properties {
 bOverrideConcurrency=true
 maxDistance=13000
 volumeMultiplier=1
 concurrencyOverrides {maxCount=3}
 concurrencySet [/Irwin/Sounds/Settings/Conc/SetOneLimitToOwner_Foley_Conc, /Irwin/Sounds/Settings/Conc/Prey_SetGroup_Foley_Conc]
}
