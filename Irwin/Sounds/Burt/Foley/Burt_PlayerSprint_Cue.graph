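# Sound cue for Burt's sprint gallop foley, apparently a rideable creature, with concurrency limits from owner and prey-group settings.
# IsLocal and IsRiddenByEnemy branches select volume, sound class and attenuation, boosting the local case and giving enemy riders dedicated class and attenuation.
# All paths share a random pitch/volume modulator over a quality switch between two mixes, each layering random regular and heavy gallops from differently sized wave pools.
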SoundCue /Irwin/Sounds/Burt/Foley/Burt_PlayerSprint_Cue
root=Branch_2
nodes [
 {id=Attenuation_1, children [Modulator_1], properties {attenuationSettings=/Irwin/Sounds/Settings/Prey_Gallop_Att}}
 {id=Attenuation_2, children [Modulator_1], properties {attenuationSettings=/Game/Sounds/Attenuation/Fort_Riding_Enemy_Att}}
 {id=Attenuation_3, children [Modulator_1], properties {attenuationSettings=&Prey_Gallop_Att}}
 {id=Attenuation_4, children [Modulator_1], properties {attenuationSettings=&Prey_Gallop_Att}}
 {id=Branch_1, children [Modulator_3,Modulator_2,SoundClass_1], properties {boolParameterName=IsRiddenByEnemy}}
 {id=Branch_2, children [Modulator_0,Branch_1,Branch_1], properties {boolParameterName=IsLocal}}
 {id=Mixer_2, children [Random_2,Random_3], properties {inputVolume [1.3,1]}}
 {id=Mixer_3, children [Random_0,Random_4], properties {inputVolume [1,1]}}
 {id=Modulator_0, children [SoundClass_4], properties {pitchMax=1, pitchMin=1, volumeMax=1.55, volumeMin=1.4}}
 {id=Modulator_1, children [QualityLevel_0], properties {pitchMax=1.1, pitchMin=0.9, volumeMax=0.9, volumeMin=0.5}}
 {id=Modulator_2, children [SoundClass_3], properties {pitchMax=1, pitchMin=1, volumeMax=0.3, volumeMin=0.3}}
 {id=Modulator_3, children [SoundClass_2], properties {pitchMax=1, pitchMin=1, volumeMax=0.7, volumeMin=0.7}}
 {id=QualityLevel_0, children [Mixer_2,Mixer_2,Mixer_3,Mixer_3]}
 {
  id=Random_0
  children [WavePlayer_0,WavePlayer_1,WavePlayer_2,WavePlayer_3]
  properties {bShouldExcludeFromBranchCulling=true, weights [1,1,1,1]}
 }
 {id=Random_2, children [WavePlayer_16, WavePlayer_17, WavePlayer_19, WavePlayer_20, WavePlayer_22], properties {weights [1,1,1,1,1]}}
 {id=Random_3, children [WavePlayer_9, WavePlayer_10, WavePlayer_11, WavePlayer_12, WavePlayer_13], properties {weights [1,1,1,1,1]}}
 {
  id=Random_4
  children [WavePlayer_14, WavePlayer_15, WavePlayer_18, WavePlayer_21]
  properties {bShouldExcludeFromBranchCulling=true, weights [1,1,1,1]}
 }
 {
  id=SoundClass_1
  children [Attenuation_1]
  properties {soundClassOverride=/Game/Sounds/FortSoundClassesAndModes/Fort_Creatures_Steps_Tactical}
 }
 {
  id=SoundClass_2
  children [Attenuation_2]
  properties {soundClassOverride=&Fort_Player_Foley_3P_Enemy_Athena}
 }
 {
  id=SoundClass_3
  children [Attenuation_3]
  properties {soundClassOverride=&Fort_Creatures_Steps_Tactical}
 }
 {
  id=SoundClass_4
  children [Attenuation_4]
  properties {soundClassOverride=&Fort_Creatures_Steps_Tactical}
 }
 {id=WavePlayer_0, properties {soundWaveAssetPtr=&/Burt_Gallup_02}}
 {id=WavePlayer_1, properties {soundWaveAssetPtr=&/Burt_Gallup_03}}
 {id=WavePlayer_10, properties {soundWaveAssetPtr=&/Burt_Gallup_Heavy_01}}
 {id=WavePlayer_11, properties {soundWaveAssetPtr=&/Burt_Gallup_Heavy_02}}
 {id=WavePlayer_12, properties {soundWaveAssetPtr=&/Burt_Gallup_Heavy_03}}
 {id=WavePlayer_13, properties {soundWaveAssetPtr=&/Burt_Gallup_Heavy_04}}
 {id=WavePlayer_14, properties {soundWaveAssetPtr=&/Burt_Gallup_Heavy_05}}
 {id=WavePlayer_15, properties {soundWaveAssetPtr=&Burt_Gallup_Heavy_01}}
 {id=WavePlayer_16, properties {soundWaveAssetPtr=&Burt_Gallup_02}}
 {id=WavePlayer_17, properties {soundWaveAssetPtr=&Burt_Gallup_03}}
 {id=WavePlayer_18, properties {soundWaveAssetPtr=&Burt_Gallup_Heavy_02}}
 {id=WavePlayer_19, properties {soundWaveAssetPtr=&/Burt_Gallup_05}}
 {id=WavePlayer_2, properties {soundWaveAssetPtr=&Burt_Gallup_05}}
 {id=WavePlayer_20, properties {soundWaveAssetPtr=&/Burt_Gallup_06}}
 {id=WavePlayer_21, properties {soundWaveAssetPtr=&Burt_Gallup_Heavy_04}}
 {id=WavePlayer_22, properties {soundWaveAssetPtr=&/Burt_Gallup_08}}
 {id=WavePlayer_3, properties {soundWaveAssetPtr=&Burt_Gallup_06}}
 {id=WavePlayer_9, properties {soundWaveAssetPtr=&Burt_Gallup_Heavy_05}}
]
properties {
 maxDistance=13000
 pitchMultiplier=1.3
 volumeMultiplier=1.1
 concurrencySet [/Irwin/Sounds/Settings/Conc/SetOneLimitToOwner_Foley_Conc, /Irwin/Sounds/Settings/Conc/Prey_SetGroup_Foley_Conc]
}
