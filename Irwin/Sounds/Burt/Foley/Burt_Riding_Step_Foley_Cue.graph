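# Sound cue for rider foley on Burt's riding steps, limited by the global Athena footstep concurrency.
# IsLocal and IsRiddenByEnemy branches pick volume, sound class and attenuation for local, enemy-ridden and other cases, with local boosted and non-enemy quieter.
# All paths share pitch and volume randomization into a quality switch between fuller and reduced variation sets.
# Each set crossfades random walk, run and sprint waves by the Speed parameter.
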SoundCue /Irwin/Sounds/Burt/Foley/Burt_Riding_Step_Foley_Cue
root=Branch_2
nodes [
 {
  id=Attenuation_1
  children [Modulator_1]
  properties {attenuationSettings=&Foley_Attenuation_Basic_Att}
 }
 {id=Attenuation_2, children [Modulator_1], properties {attenuationSettings=/Game/Sounds/Attenuation/Riding_Enemy_Foley_Att}}
 {
  id=Attenuation_3
  children [Modulator_1]
  properties {attenuationSettings=&Foley_Attenuation_Basic_Att}
 }
 {id=Attenuation_4, children [Modulator_1], properties {attenuationSettings=/Game/Sounds/Fort_Player/Footsteps/Settings/Fort_Player_FS_1P_Att}}
 {id=Branch_1, children [Modulator_3,Modulator_2,SoundClass_1], properties {boolParameterName=IsRiddenByEnemy}}
 {id=Branch_2, children [Modulator_0,Branch_1,Branch_1], properties {boolParameterName=IsLocal}}
 {id=Modulator_0, children [SoundClass_4], properties {pitchMax=1, pitchMin=1, volumeMax=1.5, volumeMin=1.25}}
 {id=Modulator_1, children [QualityLevel_1], properties {pitchMax=1.1, pitchMin=0.9, volumeMax=1, volumeMin=0.4}}
 {id=Modulator_2, children [SoundClass_3], properties {pitchMax=1, pitchMin=1, volumeMax=0.3, volumeMin=0.3}}
 {id=Modulator_3, children [SoundClass_2], properties {pitchMax=1, pitchMin=1, volumeMax=1, volumeMin=1}}
 {
  id=ParamCrossFade_0
  children [Random_2,Random_3,Random_4]
  properties {
   paramName=Speed
   crossFadeInput [
    {fadeInDistanceEnd=0, fadeInDistanceStart=0, fadeOutDistanceEnd=0.3, fadeOutDistanceStart=0.2, volume=0.4}
    {fadeInDistanceEnd=0.4, fadeInDistanceStart=0.2, fadeOutDistanceEnd=1, fadeOutDistanceStart=1, volume=0.6}
    {fadeInDistanceEnd=0.6, fadeInDistanceStart=0.5, fadeOutDistanceEnd=1.1, fadeOutDistanceStart=1, volume=0.8}
   ]
  }
 }
 {
  id=ParamCrossFade_3
  children [Random_5,Random_6,Random_7]
  properties {
   paramName=Speed
   crossFadeInput [
    {fadeInDistanceEnd=0, fadeInDistanceStart=0, fadeOutDistanceEnd=0.3, fadeOutDistanceStart=0.2, volume=0.4}
    {fadeInDistanceEnd=0.4, fadeInDistanceStart=0.2, fadeOutDistanceEnd=1, fadeOutDistanceStart=1, volume=0.6}
    {fadeInDistanceEnd=0.6, fadeInDistanceStart=0.5, fadeOutDistanceEnd=1.1, fadeOutDistanceStart=1, volume=0.8}
   ]
  }
 }
 {id=QualityLevel_1, children [ParamCrossFade_0, ParamCrossFade_0, ParamCrossFade_3, ParamCrossFade_3]}
 {
  id=Random_2
  children [WavePlayer_15, WavePlayer_16, WavePlayer_17, WavePlayer_18, WavePlayer_19, WavePlayer_20]
  properties {weights [1,1,1,1,1,1]}
 }
 {
  id=Random_3
  children [WavePlayer_0,WavePlayer_1,WavePlayer_2,WavePlayer_3,WavePlayer_4,WavePlayer_5,WavePlayer_6]
  properties {weights [1,1,1,1,1,1,1]}
 }
 {
  id=Random_4
  children [WavePlayer_7, WavePlayer_8, WavePlayer_9, WavePlayer_10, WavePlayer_11, WavePlayer_12, WavePlayer_13, WavePlayer_14]
  properties {weights [1,1,1,1,1,1,1,1]}
 }
 {
  id=Random_5
  children [WavePlayer_36, WavePlayer_37, WavePlayer_38, WavePlayer_39]
  properties {bShouldExcludeFromBranchCulling=true, weights [1,1,1,1]}
 }
 {
  id=Random_6
  children [WavePlayer_21, WavePlayer_22, WavePlayer_23, WavePlayer_24]
  properties {bShouldExcludeFromBranchCulling=true, weights [1,1,1,1]}
 }
 {
  id=Random_7
  children [WavePlayer_28, WavePlayer_29, WavePlayer_30, WavePlayer_31]
  properties {bShouldExcludeFromBranchCulling=true, weights [1,1,1,1]}
 }
 {
  id=SoundClass_1
  children [Attenuation_1]
  properties {soundClassOverride=/Game/Sounds/FortSoundClassesAndModes/Fort_Creatures_Steps_Tactical}
 }
 {
  id=SoundClass_2
  children [Attenuation_2]
  properties {soundClassOverride=&Fort_Player_Foley_3P_Enemy_Athena}
 }
 {
  id=SoundClass_3
  children [Attenuation_3]
  properties {soundClassOverride=&Fort_Creatures_Steps_Tactical}
 }
 {
  id=SoundClass_4
  children [Attenuation_4]
  properties {soundClassOverride=&Fort_Creatures_Steps_Tactical}
 }
 {id=WavePlayer_0, properties {soundWaveAssetPtr=/Irwin/Sounds/Shared/Player_Riding_Foley_Run_01}}
 {id=WavePlayer_1, properties {soundWaveAssetPtr=/Irwin/Sounds/Shared/Player_Riding_Foley_Run_02}}
 {id=WavePlayer_10, properties {soundWaveAssetPtr=/Irwin/Sounds/Shared/Player_Riding_Foley_Sprint_04}}
 {id=WavePlayer_11, properties {soundWaveAssetPtr=/Irwin/Sounds/Shared/Player_Riding_Foley_Sprint_05}}
 {id=WavePlayer_12, properties {soundWaveAssetPtr=/Irwin/Sounds/Shared/Player_Riding_Foley_Sprint_06}}
 {id=WavePlayer_13, properties {soundWaveAssetPtr=/Irwin/Sounds/Shared/Player_Riding_Foley_Sprint_07}}
 {id=WavePlayer_14, properties {soundWaveAssetPtr=/Irwin/Sounds/Shared/Player_Riding_Foley_Sprint_08}}
 {id=WavePlayer_15, properties {soundWaveAssetPtr=/Irwin/Sounds/Shared/Player_Riding_Foley_Walk_01}}
 {id=WavePlayer_16, properties {soundWaveAssetPtr=/Irwin/Sounds/Shared/Player_Riding_Foley_Walk_02}}
 {id=WavePlayer_17, properties {soundWaveAssetPtr=/Irwin/Sounds/Shared/Player_Riding_Foley_Walk_03}}
 {id=WavePlayer_18, properties {soundWaveAssetPtr=/Irwin/Sounds/Shared/Player_Riding_Foley_Walk_04}}
 {id=WavePlayer_19, properties {soundWaveAssetPtr=/Irwin/Sounds/Shared/Player_Riding_Foley_Walk_05}}
 {id=WavePlayer_2, properties {soundWaveAssetPtr=/Irwin/Sounds/Shared/Player_Riding_Foley_Run_03}}
 {id=WavePlayer_20, properties {soundWaveAssetPtr=/Irwin/Sounds/Shared/Player_Riding_Foley_Walk_06}}
 {id=WavePlayer_21, properties {soundWaveAssetPtr=&Player_Riding_Foley_Run_01}}
 {id=WavePlayer_22, properties {soundWaveAssetPtr=&Player_Riding_Foley_Run_02}}
 {id=WavePlayer_23, properties {soundWaveAssetPtr=&Player_Riding_Foley_Run_03}}
 {id=WavePlayer_24, properties {soundWaveAssetPtr=/Irwin/Sounds/Shared/Player_Riding_Foley_Run_04}}
 {id=WavePlayer_28, properties {soundWaveAssetPtr=/Irwin/Sounds/Shared/Player_Riding_Foley_Sprint_01}}
 {id=WavePlayer_29, properties {soundWaveAssetPtr=/Irwin/Sounds/Shared/Player_Riding_Foley_Sprint_02}}
 {id=WavePlayer_3, properties {soundWaveAssetPtr=&Player_Riding_Foley_Run_04}}
 {id=WavePlayer_30, properties {soundWaveAssetPtr=/Irwin/Sounds/Shared/Player_Riding_Foley_Sprint_03}}
 {id=WavePlayer_31, properties {soundWaveAssetPtr=&Player_Riding_Foley_Sprint_04}}
 {id=WavePlayer_36, properties {soundWaveAssetPtr=&Player_Riding_Foley_Walk_01}}
 {id=WavePlayer_37, properties {soundWaveAssetPtr=&Player_Riding_Foley_Walk_02}}
 {id=WavePlayer_38, properties {soundWaveAssetPtr=&Player_Riding_Foley_Walk_03}}
 {id=WavePlayer_39, properties {soundWaveAssetPtr=&Player_Riding_Foley_Walk_04}}
 {id=WavePlayer_4, properties {soundWaveAssetPtr=/Irwin/Sounds/Shared/Player_Riding_Foley_Run_05}}
 {id=WavePlayer_5, properties {soundWaveAssetPtr=/Irwin/Sounds/Shared/Player_Riding_Foley_Run_06}}
 {id=WavePlayer_6, properties {soundWaveAssetPtr=/Irwin/Sounds/Shared/Player_Riding_Foley_Run_07}}
 {id=WavePlayer_7, properties {soundWaveAssetPtr=&Player_Riding_Foley_Sprint_01}}
 {id=WavePlayer_8, properties {soundWaveAssetPtr=&Player_Riding_Foley_Sprint_02}}
 {id=WavePlayer_9, properties {soundWaveAssetPtr=&Player_Riding_Foley_Sprint_03}}
]
properties {maxDistance=5500, volumeMultiplier=0.9, concurrencySet [&Athena_Footstep_Global_Conc]}
