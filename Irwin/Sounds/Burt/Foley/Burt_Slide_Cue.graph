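# Sound cue for Burt's slide foley, playing randomized slide variations with slight pitch and volume variation.
# An IsLocal branch, then IsRiddenByEnemy on its non-local paths, selects volume, sound class and attenuation, giving enemy-ridden slides their own foley class and riding attenuation.
# All paths share one modulator feeding a quality switch between five-wave and three-wave random pools.
# Two concurrency sets limit overlapping playback.
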SoundCue /Irwin/Sounds/Burt/Foley/Burt_Slide_Cue
root=Branch_2
nodes [
 {id=Attenuation_1, children [Modulator_1], properties {attenuationSettings=/Irwin/Sounds/Settings/Prey_Step_Att}}
 {id=Attenuation_2, children [Modulator_1], properties {attenuationSettings=/Game/Sounds/Attenuation/Fort_Riding_Enemy_Att}}
 {id=Attenuation_3, children [Modulator_1], properties {attenuationSettings=&Prey_Step_Att}}
 {id=Attenuation_4, children [Modulator_1], properties {attenuationSettings=&Prey_Step_Att}}
 {id=Branch_1, children [Modulator_3,Modulator_2,SoundClass_1], properties {boolParameterName=IsRiddenByEnemy}}
 {id=Branch_2, children [SoundClass_4,Branch_1,Branch_1], properties {boolParameterName=IsLocal}}
 {id=Modulator_1, children [QualityLevel_0], properties {pitchMax=1.1, pitchMin=0.9, volumeMax=0.9, volumeMin=0.75}}
 {id=Modulator_2, children [SoundClass_3], properties {pitchMax=1, pitchMin=1, volumeMax=0.1, volumeMin=0.1}}
 {id=Modulator_3, children [SoundClass_2], properties {pitchMax=1, pitchMin=1, volumeMax=0.7, volumeMin=0.7}}
 {id=QualityLevel_0, children [Random_1,Random_1,Random_0,Random_0]}
 {id=Random_0, children [WavePlayer_1, WavePlayer_2, WavePlayer_10], properties {bShouldExcludeFromBranchCulling=true, weights [1,1,1]}}
 {id=Random_1, children [WavePlayer_0,WavePlayer_6,WavePlayer_7,WavePlayer_8,WavePlayer_9], properties {weights [1,1,1,1,1]}}
 {
  id=SoundClass_1
  children [Attenuation_1]
  properties {soundClassOverride=/Game/Sounds/FortSoundClassesAndModes/Fort_Creatures_Steps_Tactical}
 }
 {
  id=SoundClass_2
  children [Attenuation_2]
  properties {soundClassOverride=&Fort_Player_Foley_3P_Enemy_Athena}
 }
 {
  id=SoundClass_3
  children [Attenuation_3]
  properties {soundClassOverride=&Fort_Creatures_Steps_Tactical}
 }
 {
  id=SoundClass_4
  children [Attenuation_4]
  properties {soundClassOverride=&Fort_Creatures_Steps_Tactical}
 }
 {id=WavePlayer_0, properties {soundWaveAssetPtr=&/Burt_Slide_01}}
 {id=WavePlayer_1, properties {soundWaveAssetPtr=&Burt_Slide_01}}
 {id=WavePlayer_10, properties {soundWaveAssetPtr=&/Burt_Slide_05}}
 {id=WavePlayer_2, properties {soundWaveAssetPtr=&/Burt_Slide_02}}
 {id=WavePlayer_6, properties {soundWaveAssetPtr=&Burt_Slide_02}}
 {id=WavePlayer_7, properties {soundWaveAssetPtr=&/Burt_Slide_03}}
 {id=WavePlayer_8, properties {soundWaveAssetPtr=&/Burt_Slide_04}}
 {id=WavePlayer_9, properties {soundWaveAssetPtr=&Burt_Slide_05}}
]
properties {
 maxDistance=13000
 volumeMultiplier=1.5
 concurrencySet [/Irwin/Sounds/Settings/Conc/Prey_SetGroup_Foley_Conc, /Irwin/Sounds/Settings/Conc/SetOneLimitToOwner_Foley_Conc]
}
