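# Sound cue for Burt's swim-jump foley, built from randomized Grandma swim-jump waves with pitch and volume variation.
# An IsLocal branch plays local sounds directly; otherwise IsRiddenByEnemy picks a louder enemy foley class and riding attenuation when true, or heavily reduced volume when false.
# All other paths use creature-step class and prey attenuation, with an unmodulated fallback for unset parameters.
# Lower two quality levels draw on four variations, the upper two on only two.
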SoundCue /Irwin/Sounds/Burt/Foley/Burt_Swim_Jump_Cue
root=Branch_2
nodes [
 {id=Attenuation_0, children [Modulator_1], properties {attenuationSettings=/Irwin/Sounds/Settings/Prey_Step_Att}}
 {id=Attenuation_2, children [Modulator_1], properties {attenuationSettings=/Game/Sounds/Attenuation/Fort_Riding_Enemy_Att}}
 {id=Attenuation_3, children [Modulator_1], properties {attenuationSettings=&Prey_Step_Att}}
 {id=Attenuation_4, children [Modulator_1], properties {attenuationSettings=&Prey_Step_Att}}
 {id=Branch_1, children [Modulator_3,Modulator_2,SoundClass_1], properties {boolParameterName=IsRiddenByEnemy}}
 {id=Branch_2, children [SoundClass_4,Branch_1,Branch_1], properties {boolParameterName=IsLocal}}
 {id=Modulator_1, children [QualityLevel_0], properties {pitchMax=1.2, pitchMin=0.9, volumeMax=1.25, volumeMin=1}}
 {id=Modulator_2, children [SoundClass_3], properties {pitchMax=1, pitchMin=1, volumeMax=0.1, volumeMin=0.1}}
 {id=Modulator_3, children [SoundClass_2], properties {pitchMax=1, pitchMin=1, volumeMax=0.7, volumeMin=0.7}}
 {id=QualityLevel_0, children [Random_1,Random_1,Random_0,Random_0]}
 {id=Random_0, children [WavePlayer_2,WavePlayer_3], properties {bShouldExcludeFromBranchCulling=true, weights [1,1]}}
 {id=Random_1, children [WavePlayer_0, WavePlayer_1, WavePlayer_10, WavePlayer_11], properties {weights [1,1,1,1]}}
 {
  id=SoundClass_1
  children [Attenuation_0]
  properties {soundClassOverride=/Game/Sounds/FortSoundClassesAndModes/Fort_Creatures_Steps_Tactical}
 }
 {
  id=SoundClass_2
  children [Attenuation_2]
  properties {soundClassOverride=&Fort_Player_Foley_3P_Enemy_Athena}
 }
 {
  id=SoundClass_3
  children [Attenuation_3]
  properties {soundClassOverride=&Fort_Creatures_Steps_Tactical}
 }
 {
  id=SoundClass_4
  children [Attenuation_4]
  properties {soundClassOverride=&Fort_Creatures_Steps_Tactical}
 }
 {id=WavePlayer_0, properties {soundWaveAssetPtr=/Irwin/Sounds/Grandma/Foley/Foly_Grandma_Swim_Jump_01}}
 {id=WavePlayer_1, properties {soundWaveAssetPtr=/Irwin/Sounds/Grandma/Foley/Foly_Grandma_Swim_Jump_02}}
 {id=WavePlayer_10, properties {soundWaveAssetPtr=/Irwin/Sounds/Grandma/Foley/Foly_Grandma_Swim_Jump_03}}
 {id=WavePlayer_11, properties {soundWaveAssetPtr=/Irwin/Sounds/Grandma/Foley/Foly_Grandma_Swim_Jump_04}}
 {id=WavePlayer_2, properties {soundWaveAssetPtr=&Foly_Grandma_Swim_Jump_01}}
 {id=WavePlayer_3, properties {soundWaveAssetPtr=&Foly_Grandma_Swim_Jump_02}}
]
properties {maxDistance=13000, volumeMultiplier=1, concurrencySet [/Irwin/Sounds/Settings/Conc/Predator_SetGroup_Foley_Conc]}
