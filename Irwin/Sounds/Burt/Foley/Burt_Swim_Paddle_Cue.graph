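# Sound cue for Burt's swim paddle foley, built from reused Grandma swim paddle waves.
# IsLocal and IsRiddenByEnemy branches select sound class, attenuation and volume, giving enemy-ridden playback its own foley class and riding attenuation, while the non-enemy case is heavily ducked.
# All paths converge on one pitch and volume randomizer, then a quality switch between nine-wave and four-wave random pools.
# A predator foley concurrency set limits playback.
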SoundCue /Irwin/Sounds/Burt/Foley/Burt_Swim_Paddle_Cue
root=Branch_2
nodes [
 {id=Attenuation_0, children [Modulator_1], properties {attenuationSettings=/Irwin/Sounds/Settings/Prey_Step_Att}}
 {id=Attenuation_2, children [Modulator_1], properties {attenuationSettings=/Game/Sounds/Attenuation/Fort_Riding_Enemy_Att}}
 {id=Attenuation_3, children [Modulator_1], properties {attenuationSettings=&Prey_Step_Att}}
 {id=Attenuation_4, children [Modulator_1], properties {attenuationSettings=&Prey_Step_Att}}
 {id=Branch_1, children [Modulator_3,Modulator_2,SoundClass_0], properties {boolParameterName=IsRiddenByEnemy}}
 {id=Branch_2, children [SoundClass_4,Branch_1,Branch_1], properties {boolParameterName=IsLocal}}
 {id=Modulator_1, children [QualityLevel_0], properties {pitchMax=1.25, pitchMin=0.9, volumeMax=1, volumeMin=0.9}}
 {id=Modulator_2, children [SoundClass_3], properties {pitchMax=1, pitchMin=1, volumeMax=0.1, volumeMin=0.1}}
 {id=Modulator_3, children [SoundClass_2], properties {pitchMax=1, pitchMin=1, volumeMax=0.7, volumeMin=0.7}}
 {id=QualityLevel_0, children [Random_0,Random_0,Random_1,Random_1]}
 {
  id=Random_0
  children [WavePlayer_2, WavePlayer_3, WavePlayer_4, WavePlayer_5, WavePlayer_6, WavePlayer_7, WavePlayer_8, WavePlayer_9, WavePlayer_14]
  properties {bShouldExcludeFromBranchCulling=true, weights [1,1,1,1,1,1,1,1,1]}
 }
 {
  id=Random_1
  children [WavePlayer_0, WavePlayer_1, WavePlayer_10, WavePlayer_11]
  properties {bShouldExcludeFromBranchCulling=true, weights [1,1,1,1]}
 }
 {
  id=SoundClass_0
  children [Attenuation_0]
  properties {soundClassOverride=/Game/Sounds/FortSoundClassesAndModes/Fort_Creatures_Steps_Tactical}
 }
 {
  id=SoundClass_2
  children [Attenuation_2]
  properties {soundClassOverride=&Fort_Player_Foley_3P_Enemy_Athena}
 }
 {
  id=SoundClass_3
  children [Attenuation_3]
  properties {soundClassOverride=&Fort_Creatures_Steps_Tactical}
 }
 {
  id=SoundClass_4
  children [Attenuation_4]
  properties {soundClassOverride=&Fort_Creatures_Steps_Tactical}
 }
 {id=WavePlayer_0, properties {soundWaveAssetPtr=/Irwin/Sounds/Grandma/Foley/Foly_Grandma_Swim_Paddle_06}}
 {id=WavePlayer_1, properties {soundWaveAssetPtr=/Irwin/Sounds/Grandma/Foley/Foly_Grandma_Swim_Paddle_07}}
 {id=WavePlayer_10, properties {soundWaveAssetPtr=/Irwin/Sounds/Grandma/Foley/Foly_Grandma_Swim_Paddle_08}}
 {id=WavePlayer_11, properties {soundWaveAssetPtr=/Irwin/Sounds/Grandma/Foley/Foly_Grandma_Swim_Paddle_09}}
 {id=WavePlayer_14, properties {soundWaveAssetPtr=&Foly_Grandma_Swim_Paddle_09}}
 {id=WavePlayer_2, properties {soundWaveAssetPtr=/Irwin/Sounds/Grandma/Foley/Foly_Grandma_Swim_Paddle_01}}
 {id=WavePlayer_3, properties {soundWaveAssetPtr=/Irwin/Sounds/Grandma/Foley/Foly_Grandma_Swim_Paddle_02}}
 {id=WavePlayer_4, properties {soundWaveAssetPtr=/Irwin/Sounds/Grandma/Foley/Foly_Grandma_Swim_Paddle_03}}
 {id=WavePlayer_5, properties {soundWaveAssetPtr=/Irwin/Sounds/Grandma/Foley/Foly_Grandma_Swim_Paddle_04}}
 {id=WavePlayer_6, properties {soundWaveAssetPtr=/Irwin/Sounds/Grandma/Foley/Foly_Grandma_Swim_Paddle_05}}
 {id=WavePlayer_7, properties {soundWaveAssetPtr=&Foly_Grandma_Swim_Paddle_06}}
 {id=WavePlayer_8, properties {soundWaveAssetPtr=&Foly_Grandma_Swim_Paddle_07}}
 {id=WavePlayer_9, properties {soundWaveAssetPtr=&Foly_Grandma_Swim_Paddle_08}}
]
properties {maxDistance=13000, volumeMultiplier=1.5, concurrencySet [/Irwin/Sounds/Settings/Conc/Predator_SetGroup_Foley_Conc]}
